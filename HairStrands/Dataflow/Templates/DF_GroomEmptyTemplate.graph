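# A minimal Dataflow template for HairStrands grooms that passes a groom asset's guides and strands through to a groom terminal without edits.
# GetGroomAsset feeds two GroomAssetToCollection nodes, one set to Guides and the other to Strands.
# Their collections connect to the matching guide and strand inputs of GroomAssetTerminal, with no processing nodes between.
# The groom asset reference and the terminal's attribute names are left empty.
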Dataflow /HairStrands/Dataflow/Templates/DF_GroomEmptyTemplate
name=DF_GroomEmptyTemplate
connections [
 {
  fromNode=GetGroomAsset
  fromPin=GroomAsset
  fromPinGuid=#1
  toNode=GroomAssetToCollection
  toPin=GroomAsset
  toPinGuid=#2
 }
 {
  fromNode=GetGroomAsset
  fromPin=GroomAsset
  fromPinGuid=#1
  toNode=GroomAssetToCollection_0
  toPin=GroomAsset
  toPinGuid=#3
 }
 {
  fromNode=GroomAssetToCollection
  fromPin=Collection
  fromPinGuid=#4
  toNode=GroomAssetTerminal
  toPin=GuidesCollection
  toPinGuid=#5
 }
 {
  fromNode=GroomAssetToCollection_0
  fromPin=Collection
  fromPinGuid=#6
  toNode=GroomAssetTerminal
  toPin=StrandsCollection
  toPinGuid=#7
 }
]
nodes [
 GetGroomAsset {
  type=FGetGroomAssetDataflowNode_v2
  guid=#8
  inputPins []
  outputPins [GroomAsset {type="TObjectPtr<UGroomAsset>"}]
  properties {GroomAsset=None}
 }
 GroomAssetToCollection {
  type=FGroomAssetToCollectionDataflowNode
  guid=#9
  inputPins [
   GroomAsset {type="TObjectPtr<UGroomAsset>"}
   CurvesThickness {type=float}
   CurvesType {type=EGroomCollectionType}
  ]
  outputPins [Collection {type=FManagedArrayCollection}]
  properties {CurvesThickness=0.5, CurvesType=Guides, GroomAsset=None}
 }
 GroomAssetTerminal {
  type=FGroomAssetTerminalDataflowNode_v2
  guid=#10
  inputPins [
   GuidesCollection {type=FManagedArrayCollection}
   StrandsCollection {type=FManagedArrayCollection}
  ]
  outputPins []
  properties {AttributeNames=""}
 }
 GroomAssetToCollection_0 {
  type=FGroomAssetToCollectionDataflowNode
  guid=#11
  inputPins [
   GroomAsset {type="TObjectPtr<UGroomAsset>"}
   CurvesThickness {type=float}
   CurvesType {type=EGroomCollectionType}
  ]
  outputPins [Collection {type=FManagedArrayCollection}]
  properties {CurvesThickness=0.5, CurvesType=Strands, GroomAsset=None}
 }
]
