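# Groom Dataflow template that generates simulation-ready guides from a groom asset's strands and writes guides and strands back through the groom asset terminal.
# Variable-driven nodes generate, resample and smooth the guides, then roots are attached and kinematic weights exposed for painting.
# Curve LODs are built, then skin weights transfer from the SkeletalMesh variable into an edit node.
# Strands pass through unchanged; the terminal also receives bone, parent, LOD and kinematic attribute keys.
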Dataflow /HairStrands/Dataflow/Templates/DF_GroomGuidesTemplate
name=DF_GroomGuidesTemplate
connections [
 {
  fromNode=BuildCurveLODs
  fromPin=CurveLodsKey
  fromPinGuid=#1
  toNode=ReRouteNode_3
  toPin=Value
  toPinGuid=#2
 }
 {
  fromNode=ReRouteNode_3
  fromPin=Value
  fromPinGuid=#3
  toNode=ReRouteNode_4
  toPin=Value
  toPinGuid=#4
 }
 {
  fromNode=TransferGeometrySkinWeights
  fromPin=BoneWeightsKey
  fromPinGuid=#5
  toNode=EditSkinWeights
  toPin=BoneWeightsKey
  toPinGuid=#6
 }
 {
  fromNode=TransferGeometrySkinWeights
  fromPin=BoneIndicesKey
  fromPinGuid=#7
  toNode=EditSkinWeights
  toPin=BoneIndicesKey
  toPinGuid=#8
 }
 {
  fromNode=SkeletalMesh
  fromPin=Value
  fromPinGuid=#9
  toNode=TransferGeometrySkinWeights
  toPin=SkeletalMesh
  toPinGuid=#10
 }
 {
  fromNode=AttachCurveRoots
  fromPin=KinematicWeightsKey
  fromPinGuid=#11
  toNode=PaintWeightMap
  toPin=AttributeKey
  toPinGuid=#12
 }
 {
  fromNode=GroomAssetToCollection_0
  fromPin=Collection
  fromPinGuid=#13
  toNode=GenerateCurveGeometry
  toPin=SourceCurves
  toPinGuid=#14
 }
 {
  fromNode=GenerateCurveGeometry
  fromPin=Collection
  fromPinGuid=#15
  toNode=ResampleCurvePoints
  toPin=Collection
  toPinGuid=#16
 }
 {
  fromNode=SmoothCurvePoints
  fromPin=Collection
  fromPinGuid=#17
  toNode=AttachCurveRoots
  toPin=Collection
  toPinGuid=#18
 }
 {
  fromNode=GuideSmoothFactor
  fromPin=Value
  fromPinGuid=#19
  toNode=SmoothCurvePoints
  toPin=SmoothingFactor
  toPinGuid=#20
 }
 {
  fromNode=BuildCurveLODs
  fromPin=Collection
  fromPinGuid=#21
  toNode=TransferGeometrySkinWeights
  toPin=Collection
  toPinGuid=#22
 }
 {
  fromNode=ResampleCurvePoints
  fromPin=Collection
  fromPinGuid=#23
  toNode=SmoothCurvePoints
  toPin=Collection
  toPinGuid=#24
 }
 {
  fromNode=NumGroomGuides
  fromPin=Value
  fromPinGuid=#25
  toNode=GenerateCurveGeometry
  toPin=CurveCount
  toPinGuid=#26
 }
 {
  fromNode=GroomAssetToCollection
  fromPin=Collection
  fromPinGuid=#27
  toNode=GenerateCurveGeometry
  toPin=Collection
  toPinGuid=#28
 }
 {
  fromNode=NumGuidePoints
  fromPin=Value
  fromPinGuid=#29
  toNode=ResampleCurvePoints
  toPin=NumPoints
  toPinGuid=#30
 }
 {
  fromNode=SkeletalMesh
  fromPin=Value
  fromPinGuid=#9
  toNode=EditSkinWeights
  toPin=SkeletalMesh
  toPinGuid=#31
 }
 {
  fromNode=TransferGeometrySkinWeights
  fromPin=Collection
  fromPinGuid=#32
  toNode=EditSkinWeights
  toPin=Collection
  toPinGuid=#33
 }
 {
  fromNode=GroomAssetToCollection_0
  fromPin=Collection
  fromPinGuid=#13
  toNode=ReRouteNode
  toPin=Value
  toPinGuid=#34
 }
 {
  fromNode=ReRouteNode
  fromPin=Value
  fromPinGuid=#35
  toNode=ReRouteNode_0
  toPin=Value
  toPinGuid=#36
 }
 {
  fromNode=GetGroomAsset
  fromPin=GroomAsset
  fromPinGuid=#37
  toNode=GroomAssetToCollection
  toPin=GroomAsset
  toPinGuid=#38
 }
 {
  fromNode=GetGroomAsset
  fromPin=GroomAsset
  fromPinGuid=#37
  toNode=GroomAssetToCollection_0
  toPin=GroomAsset
  toPinGuid=#39
 }
 {
  fromNode=AttachCurveRoots
  fromPin=Collection
  fromPinGuid=#40
  toNode=PaintWeightMap
  toPin=Collection
  toPinGuid=#41
 }
 {
  fromNode=PaintWeightMap
  fromPin=Collection
  fromPinGuid=#42
  toNode=BuildCurveLODs
  toPin=Collection
  toPinGuid=#43
 }
 {
  fromNode=ReRouteNode_5
  fromPin=Value
  fromPinGuid=#44
  toNode=ReRouteNode_6
  toPin=Value
  toPinGuid=#45
 }
 {
  fromNode=AttachCurveRoots
  fromPin=KinematicWeightsKey
  fromPinGuid=#11
  toNode=ReRouteNode_5
  toPin=Value
  toPinGuid=#46
 }
 {
  fromNode=BuildCurveLODs
  fromPin=CurveParentsKey
  fromPinGuid=#47
  toNode=ReRouteNode_1
  toPin=Value
  toPinGuid=#48
 }
 {
  fromNode=ReRouteNode_1
  fromPin=Value
  fromPinGuid=#49
  toNode=ReRouteNode_2
  toPin=Value
  toPinGuid=#50
 }
 {
  fromNode=EditSkinWeights
  fromPin=Collection
  fromPinGuid=#51
  toNode=GroomAssetTerminal
  toPin=GuidesCollection
  toPinGuid=#52
 }
 {
  fromNode=ReRouteNode_0
  fromPin=Value
  fromPinGuid=#53
  toNode=GroomAssetTerminal
  toPin=StrandsCollection
  toPinGuid=#54
 }
 {
  fromNode=EditSkinWeights
  fromPin=BoneIndicesKey
  fromPinGuid=#55
  toNode=GroomAssetTerminal
  toPin=Bone Indices
  toPinGuid=#56
 }
 {
  fromNode=EditSkinWeights
  fromPin=BoneWeightsKey
  fromPinGuid=#57
  toNode=GroomAssetTerminal
  toPin=Bone Weights
  toPinGuid=#58
 }
 {
  fromNode=ReRouteNode_2
  fromPin=Value
  fromPinGuid=#59
  toNode=GroomAssetTerminal
  toPin=Curve Parents
  toPinGuid=#60
 }
 {
  fromNode=ReRouteNode_4
  fromPin=Value
  fromPinGuid=#61
  toNode=GroomAssetTerminal
  toPin=Curve LODs
  toPinGuid=#62
 }
 {
  fromNode=ReRouteNode_6
  fromPin=Value
  fromPinGuid=#63
  toNode=GroomAssetTerminal
  toPin=Kinematic Weights
  toPinGuid=#64
 }
]
nodes [
 GroomAssetToCollection {
  type=FGroomAssetToCollectionDataflowNode
  guid=#65
  inputPins [
   GroomAsset {type="TObjectPtr<UGroomAsset>"}
   CurvesThickness {type=float}
   CurvesType {type=EGroomCollectionType}
  ]
  outputPins [Collection {type=FManagedArrayCollection}]
  properties {CurvesThickness=0.5, CurvesType=Guides, GroomAsset=None}
 }
 GenerateCurveGeometry {
  type=FGenerateCurveGeometryDataflowNode
  guid=#66
  inputPins [
   Collection {type=FManagedArrayCollection}
   SourceCurves {type=FManagedArrayCollection}
   CurveSelection {type=FDataflowCurveSelection}
   CurveCount {type=int32}
  ]
  outputPins [Collection {type=FManagedArrayCollection}]
  properties {CurveCount=300, bMergeCurves=False}
 }
 GroomAssetToCollection_0 {
  type=FGroomAssetToCollectionDataflowNode
  guid=#67
  inputPins [
   GroomAsset {type="TObjectPtr<UGroomAsset>"}
   CurvesThickness {type=float}
   CurvesType {type=EGroomCollectionType}
  ]
  outputPins [Collection {type=FManagedArrayCollection}]
  properties {CurvesThickness=0.5, CurvesType=Strands, GroomAsset=None}
 }
 NumGroomGuides {
  type=FGetDataflowVariableNode
  guid=#68
  outputPins [Value {type=int32}]
  properties {VariableName=NumGroomGuides, VariablePropertyBag=(Value=/Engine/Transient.PropertyBag_805d687dc0c11087(NumGroomGuides=300))}
 }
 ResampleCurvePoints {
  type=FResampleCurvePointsDataflowNode
  guid=#69
  inputPins [
   Collection {type=FManagedArrayCollection}
   CurveSelection {type=FDataflowCurveSelection}
   NumPoints {type=int32}
  ]
  outputPins [Collection {type=FManagedArrayCollection}]
  properties {NumPoints=16, PointsCount=Size16, bResampleForSimulation=True}
 }
 NumGuidePoints {
  type=FGetDataflowVariableNode
  guid=#70
  outputPins [Value {type=int32}]
  properties {VariableName=NumGuidePoints, VariablePropertyBag=(Value=/Engine/Transient.PropertyBag_5b516c56c5f604db(NumGuidePoints=16))}
 }
 SmoothCurvePoints {
  type=FSmoothCurvePointsDataflowNode
  guid=#71
  inputPins [
   Collection {type=FManagedArrayCollection}
   CurveSelection {type=FDataflowCurveSelection}
   SmoothingFactor {type=float}
  ]
  outputPins [Collection {type=FManagedArrayCollection}]
  properties {SmoothingFactor=0}
 }
 GuideSmoothFactor {
  type=FGetDataflowVariableNode
  guid=#72
  outputPins [Value {type=float}]
  properties {VariableName=GuideSmoothFactor, VariablePropertyBag=(Value=/Engine/Transient.PropertyBag_f5c673accfff512e(GuideSmoothFactor=0.1))}
 }
 AttachCurveRoots {
  type=FAttachCurveRootsDataflowNode
  guid=#73
  inputPins [
   Collection {type=FManagedArrayCollection}
   CurveSelection {type=FDataflowCurveSelection}
  ]
  outputPins [
   Collection {type=FManagedArrayCollection}
   KinematicWeightsKey {type=FCollectionAttributeKey}
  ]
 }
 BuildCurveLODs {
  type=FBuildCurveLODsDataflowNode
  guid=#74
  inputPins [
   Collection {type=FManagedArrayCollection}
   CurveSelection {type=FDataflowCurveSelection}
  ]
  outputPins [
   Collection {type=FManagedArrayCollection}
   CurveParentsKey {type=FCollectionAttributeKey}
   CurveLodsKey {type=FCollectionAttributeKey}
  ]
 }
 TransferGeometrySkinWeights {
  type=FTransferGeometrySkinWeightsDataflowNode
  guid=#75
  inputPins [
   Collection {type=FManagedArrayCollection}
   VertexSelection {type=FDataflowVertexSelection}
   SkeletalMesh {type="TObjectPtr<USkeletalMesh>"}
   LODIndex {type=int32}
   RelativeTransform {type=FTransform}
  ]
  outputPins [
   Collection {type=FManagedArrayCollection}
   BoneIndicesKey {type=FCollectionAttributeKey}
   BoneWeightsKey {type=FCollectionAttributeKey}
  ]
  properties {LODIndex=0, RelativeTransform=(Rotation=(X=0,Y=0,Z=0,W=1),Translation=(X=0,Y=0,Z=0),Scale3D=(X=1,Y=1,Z=1)), SkeletalMesh=None}
 }
 SkeletalMesh {
  type=FGetDataflowVariableNode
  guid=#76
  outputPins [Value {type="TObjectPtr<USkeletalMesh>"}]
  properties {VariableName=SkeletalMesh, VariablePropertyBag=(Value=/Engine/Transient.PropertyBag_927204fe0ac9c40(SkeletalMesh=None))}
 }
 ReRouteNode {
  type=FDataflowReRouteNode
  guid=#77
  inputPins [Value {type=FManagedArrayCollection}]
  outputPins [Value {type=FManagedArrayCollection}]
 }
 ReRouteNode_0 {
  type=FDataflowReRouteNode
  guid=#78
  inputPins [Value {type=FManagedArrayCollection}]
  outputPins [Value {type=FManagedArrayCollection}]
 }
 GetGroomAsset {
  type=FGetGroomAssetDataflowNode_v2
  guid=#79
  inputPins []
  outputPins [GroomAsset {type="TObjectPtr<UGroomAsset>"}]
  properties {GroomAsset=None}
 }
 PaintWeightMap {
  type=FDataflowCollectionAddScalarVertexPropertyNode
  guid=#80
  inputPins [
   Collection {type=FManagedArrayCollection}
   AttributeKey {type=FCollectionAttributeKey}
  ]
  outputPins [
   Collection {type=FManagedArrayCollection}
   AttributeKey {type=FCollectionAttributeKey}
  ]
  properties {Name="", OverrideType=ReplaceAll, TargetGroup=(Name="Vertices")}
 }
 ReRouteNode_1 {
  type=FDataflowReRouteNode
  guid=#81
  inputPins [Value {type=FCollectionAttributeKey}]
  outputPins [Value {type=FCollectionAttributeKey}]
 }
 ReRouteNode_2 {
  type=FDataflowReRouteNode
  guid=#82
  inputPins [Value {type=FCollectionAttributeKey}]
  outputPins [Value {type=FCollectionAttributeKey}]
 }
 ReRouteNode_3 {
  type=FDataflowReRouteNode
  guid=#83
  inputPins [Value {type=FCollectionAttributeKey}]
  outputPins [Value {type=FCollectionAttributeKey}]
 }
 ReRouteNode_4 {
  type=FDataflowReRouteNode
  guid=#84
  inputPins [Value {type=FCollectionAttributeKey}]
  outputPins [Value {type=FCollectionAttributeKey}]
 }
 ReRouteNode_5 {
  type=FDataflowReRouteNode
  guid=#85
  inputPins [Value {type=FCollectionAttributeKey}]
  outputPins [Value {type=FCollectionAttributeKey}]
 }
 ReRouteNode_6 {
  type=FDataflowReRouteNode
  guid=#86
  inputPins [Value {type=FCollectionAttributeKey}]
  outputPins [Value {type=FCollectionAttributeKey}]
 }
 EditSkinWeights {
  type=FDataflowCollectionEditSkinWeightsNode
  guid=#87
  inputPins [
   Collection {type=FManagedArrayCollection}
   BoneIndicesKey {type=FCollectionAttributeKey}
   BoneWeightsKey {type=FCollectionAttributeKey}
   SkeletalMesh {type="TObjectPtr<USkeletalMesh>"}
  ]
  outputPins [
   Collection {type=FManagedArrayCollection}
   BoneIndicesKey {type=FCollectionAttributeKey}
   BoneWeightsKey {type=FCollectionAttributeKey}
  ]
  properties {BoneIndicesName="", BoneWeightsName="", SkeletalMesh=None, VertexGroup=(Name="Vertices"), bCompressSkinWeights=False}
 }
 GroomAssetTerminal {
  type=FGroomAssetTerminalDataflowNode_v2
  guid=#88
  inputPins [
   GuidesCollection {type=FManagedArrayCollection}
   StrandsCollection {type=FManagedArrayCollection}
   Bone Indices {type=FCollectionAttributeKey}
   Bone Weights {type=FCollectionAttributeKey}
   Curve Parents {type=FCollectionAttributeKey}
   Curve LODs {type=FCollectionAttributeKey}
   Kinematic Weights {type=FCollectionAttributeKey}
  ]
  outputPins []
  properties {AttributeNames=("Bone Indices","Bone Weights","Curve Parents","Curve LODs","Kinematic Weights")}
 }
]
