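# Dataflow template that builds linear skinning data for groom strands, binding them to the bones of a skeletal mesh.
# The groom asset is converted into separate guides and strands collections.
# Strands receive skin weights transferred from the SkeletalMesh variable, then pass through an Edit Skin Weights node that uses the same mesh.
# The groom terminal takes the unmodified guides, the skinned strands, and the bone index and weight attribute keys.
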Dataflow /HairStrands/Dataflow/Templates/DF_GroomStrandsLinearSkinningTemplate
name=DF_GroomStrandsLinearSkinningTemplate
connections [
 {
  fromNode=GetGroomAsset
  fromPin=GroomAsset
  fromPinGuid=#1
  toNode=GroomAssetToCollection_0
  toPin=GroomAsset
  toPinGuid=#2
 }
 {
  fromNode=TransferGeometrySkinWeights
  fromPin=Collection
  fromPinGuid=#3
  toNode=EditSkinWeights
  toPin=Collection
  toPinGuid=#4
 }
 {
  fromNode=TransferGeometrySkinWeights
  fromPin=BoneWeightsKey
  fromPinGuid=#5
  toNode=EditSkinWeights
  toPin=BoneWeightsKey
  toPinGuid=#6
 }
 {
  fromNode=TransferGeometrySkinWeights
  fromPin=BoneIndicesKey
  fromPinGuid=#7
  toNode=EditSkinWeights
  toPin=BoneIndicesKey
  toPinGuid=#8
 }
 {
  fromNode=SkeletalMesh
  fromPin=Value
  fromPinGuid=#9
  toNode=TransferGeometrySkinWeights
  toPin=SkeletalMesh
  toPinGuid=#10
 }
 {
  fromNode=ReRouteNode_v1
  fromPin=Value
  fromPinGuid=#11
  toNode=EditSkinWeights
  toPin=SkeletalMesh
  toPinGuid=#12
 }
 {
  fromNode=SkeletalMesh
  fromPin=Value
  fromPinGuid=#9
  toNode=ReRouteNode_v1
  toPin=Value
  toPinGuid=#13
 }
 {
  fromNode=GroomAssetToCollection
  fromPin=Collection
  fromPinGuid=#14
  toNode=GroomAssetTerminal
  toPin=GuidesCollection
  toPinGuid=#15
 }
 {
  fromNode=EditSkinWeights
  fromPin=Collection
  fromPinGuid=#16
  toNode=GroomAssetTerminal
  toPin=StrandsCollection
  toPinGuid=#17
 }
 {
  fromNode=EditSkinWeights
  fromPin=BoneWeightsKey
  fromPinGuid=#18
  toNode=GroomAssetTerminal
  toPin=Bone Weights
  toPinGuid=#19
 }
 {
  fromNode=GetGroomAsset
  fromPin=GroomAsset
  fromPinGuid=#1
  toNode=GroomAssetToCollection
  toPin=GroomAsset
  toPinGuid=#20
 }
 {
  fromNode=EditSkinWeights
  fromPin=BoneIndicesKey
  fromPinGuid=#21
  toNode=GroomAssetTerminal
  toPin=Bone Indices
  toPinGuid=#22
 }
 {
  fromNode=GroomAssetToCollection_0
  fromPin=Collection
  fromPinGuid=#23
  toNode=TransferGeometrySkinWeights
  toPin=Collection
  toPinGuid=#24
 }
]
nodes [
 GroomAssetToCollection {
  type=FGroomAssetToCollectionDataflowNode
  guid=#25
  inputPins [
   GroomAsset {type="TObjectPtr<UGroomAsset>"}
   CurvesThickness {type=float}
   CurvesType {type=EGroomCollectionType}
  ]
  outputPins [Collection {type=FManagedArrayCollection}]
  properties {CurvesThickness=0.5, CurvesType=Guides, GroomAsset=None}
 }
 GroomAssetToCollection_0 {
  type=FGroomAssetToCollectionDataflowNode
  guid=#26
  inputPins [
   GroomAsset {type="TObjectPtr<UGroomAsset>"}
   CurvesThickness {type=float}
   CurvesType {type=EGroomCollectionType}
  ]
  outputPins [Collection {type=FManagedArrayCollection}]
  properties {CurvesThickness=0.5, CurvesType=Strands, GroomAsset=None}
 }
 TransferGeometrySkinWeights {
  type=FTransferGeometrySkinWeightsDataflowNode
  guid=#27
  inputPins [
   Collection {type=FManagedArrayCollection}
   VertexSelection {type=FDataflowVertexSelection}
   SkeletalMesh {type="TObjectPtr<USkeletalMesh>"}
   LODIndex {type=int32}
   RelativeTransform {type=FTransform}
  ]
  outputPins [
   Collection {type=FManagedArrayCollection}
   BoneIndicesKey {type=FCollectionAttributeKey}
   BoneWeightsKey {type=FCollectionAttributeKey}
  ]
  properties {LODIndex=0, RelativeTransform=(Rotation=(X=0,Y=0,Z=0,W=1),Translation=(X=0,Y=0,Z=0),Scale3D=(X=1,Y=1,Z=1)), SkeletalMesh=None}
 }
 SkeletalMesh {
  type=FGetDataflowVariableNode
  guid=#28
  outputPins [Value {type="TObjectPtr<USkeletalMesh>"}]
  properties {VariableName=SkeletalMesh, VariablePropertyBag=(Value=/Engine/Transient.PropertyBag_927204fe0ac9c40(SkeletalMesh=None))}
 }
 GetGroomAsset {
  type=FGetGroomAssetDataflowNode_v2
  guid=#29
  inputPins []
  outputPins [GroomAsset {type="TObjectPtr<UGroomAsset>"}]
  properties {GroomAsset=None}
 }
 EditSkinWeights {
  type=FDataflowCollectionEditSkinWeightsNode
  guid=#30
  inputPins [
   Collection {type=FManagedArrayCollection}
   BoneIndicesKey {type=FCollectionAttributeKey}
   BoneWeightsKey {type=FCollectionAttributeKey}
   SkeletalMesh {type="TObjectPtr<USkeletalMesh>"}
  ]
  outputPins [
   Collection {type=FManagedArrayCollection}
   BoneIndicesKey {type=FCollectionAttributeKey}
   BoneWeightsKey {type=FCollectionAttributeKey}
  ]
  properties {BoneIndicesName="", BoneWeightsName="", SkeletalMesh=None, VertexGroup=(Name="Vertices"), bCompressSkinWeights=False}
 }
 GroomAssetTerminal {
  type=FGroomAssetTerminalDataflowNode_v2
  guid=#31
  inputPins [
   GuidesCollection {type=FManagedArrayCollection}
   StrandsCollection {type=FManagedArrayCollection}
   Bone Indices {type=FCollectionAttributeKey}
   Bone Weights {type=FCollectionAttributeKey}
  ]
  outputPins []
  properties {AttributeNames=("Bone Indices","Bone Weights")}
 }
 ReRouteNode_v1 {
  type=FDataflowReRouteNode
  guid=#32
  inputPins [Value {type="TObjectPtr<USkeletalMesh>"}]
  outputPins [Value {type="TObjectPtr<USkeletalMesh>"}]
 }
]
