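# Dataflow template that computes spline skinning for groom strands against a skeletal mesh and writes it to the groom asset.
# BuildSplineSkinWeights derives spline parameter and bone attributes for the strands from the Skel_Mesh variable and a root bone, defaulting to Braid_Start.
# These convert to linear bone indices and weights for EditSkinWeights, then back to spline form for the terminal.
# Guides go straight to the terminal without skinning.
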Dataflow /HairStrands/Dataflow/Templates/DF_GroomStrandsSplineSkinningTemplate
name=DF_GroomStrandsSplineSkinningTemplate
connections [
 {
  fromNode=GetGroomAsset
  fromPin=GroomAsset
  fromPinGuid=#1
  toNode=GroomAssetToCollection
  toPin=GroomAsset
  toPinGuid=#2
 }
 {
  fromNode=LinearToSplineSkinWeights
  fromPin=SplineBonesKey
  fromPinGuid=#3
  toNode=GroomAssetTerminal
  toPin=Spline Bones
  toPinGuid=#4
 }
 {
  fromNode=GroomAssetToCollection
  fromPin=Collection
  fromPinGuid=#5
  toNode=BuildSplineSkinWeights
  toPin=Collection
  toPinGuid=#6
 }
 {
  fromNode=EditSkinWeights
  fromPin=Collection
  fromPinGuid=#7
  toNode=LinearToSplineSkinWeights
  toPin=Collection
  toPinGuid=#8
 }
 {
  fromNode=GetGroomAsset
  fromPin=GroomAsset
  fromPinGuid=#1
  toNode=GroomAssetToCollection_0
  toPin=GroomAsset
  toPinGuid=#9
 }
 {
  fromNode=Skel_Mesh
  fromPin=Value
  fromPinGuid=#10
  toNode=BuildSplineSkinWeights
  toPin=SkeletalMesh
  toPinGuid=#11
 }
 {
  fromNode=BuildSplineSkinWeights
  fromPin=Collection
  fromPinGuid=#12
  toNode=SplineToLinearSkinWeights
  toPin=Collection
  toPinGuid=#13
 }
 {
  fromNode=SplineToLinearSkinWeights
  fromPin=Collection
  fromPinGuid=#14
  toNode=EditSkinWeights
  toPin=Collection
  toPinGuid=#15
 }
 {
  fromNode=LinearToSplineSkinWeights
  fromPin=SplineParamKey
  fromPinGuid=#16
  toNode=GroomAssetTerminal
  toPin=Spline Parameter
  toPinGuid=#17
 }
 {
  fromNode=SplineToLinearSkinWeights
  fromPin=BoneIndicesKey
  fromPinGuid=#18
  toNode=EditSkinWeights
  toPin=BoneIndicesKey
  toPinGuid=#19
 }
 {
  fromNode=SplineToLinearSkinWeights
  fromPin=BoneWeightsKey
  fromPinGuid=#20
  toNode=EditSkinWeights
  toPin=BoneWeightsKey
  toPinGuid=#21
 }
 {
  fromNode=LinearToSplineSkinWeights
  fromPin=Collection
  fromPinGuid=#22
  toNode=GroomAssetTerminal
  toPin=StrandsCollection
  toPinGuid=#23
 }
 {
  fromNode=Skel_Mesh
  fromPin=Value
  fromPinGuid=#10
  toNode=ReRouteNode_v1
  toPin=Value
  toPinGuid=#24
 }
 {
  fromNode=GroomAssetToCollection_0
  fromPin=Collection
  fromPinGuid=#25
  toNode=GroomAssetTerminal
  toPin=GuidesCollection
  toPinGuid=#26
 }
 {
  fromNode=ReRouteNode_v1
  fromPin=Value
  fromPinGuid=#27
  toNode=EditSkinWeights
  toPin=SkeletalMesh
  toPinGuid=#28
 }
 {
  fromNode=EditSkinWeights
  fromPin=BoneIndicesKey
  fromPinGuid=#29
  toNode=LinearToSplineSkinWeights
  toPin=BoneIndicesKey
  toPinGuid=#30
 }
 {
  fromNode=EditSkinWeights
  fromPin=BoneWeightsKey
  fromPinGuid=#31
  toNode=LinearToSplineSkinWeights
  toPin=BoneWeightsKey
  toPinGuid=#32
 }
 {
  fromNode=BuildSplineSkinWeights
  fromPin=SplineParamKey
  fromPinGuid=#33
  toNode=SplineToLinearSkinWeights
  toPin=SplineParamKey
  toPinGuid=#34
 }
 {
  fromNode=BuildSplineSkinWeights
  fromPin=SplineBonesKey
  fromPinGuid=#35
  toNode=SplineToLinearSkinWeights
  toPin=SplineBonesKey
  toPinGuid=#36
 }
]
nodes [
 GetGroomAsset {
  type=FGetGroomAssetDataflowNode_v2
  guid=#37
  inputPins []
  outputPins [GroomAsset {type="TObjectPtr<UGroomAsset>"}]
  properties {GroomAsset=None}
 }
 GroomAssetToCollection {
  type=FGroomAssetToCollectionDataflowNode
  guid=#38
  inputPins [
   GroomAsset {type="TObjectPtr<UGroomAsset>"}
   CurvesThickness {type=float}
   CurvesType {type=EGroomCollectionType}
  ]
  outputPins [Collection {type=FManagedArrayCollection}]
  properties {CurvesThickness=0.5, CurvesType=Strands, GroomAsset=None}
 }
 BuildSplineSkinWeights {
  type=FBuildSplineSkinWeightsDataflowNode
  guid=#39
  inputPins [
   Collection {type=FManagedArrayCollection}
   VertexSelection {type=FDataflowVertexSelection}
   SkeletalMesh {type="TObjectPtr<USkeletalMesh>"}
   LODIndex {type=int32}
   RootBones {type="TArray<FString>"}
   SamplesPerSegment {type=int32}
  ]
  outputPins [
   Collection {type=FManagedArrayCollection}
   SplineParamKey {type=FCollectionAttributeKey}
   SplineBonesKey {type=FCollectionAttributeKey}
  ]
  properties {LODIndex=0, RootBones=("Braid_Start"), SamplesPerSegment=64, SkeletalMesh=None}
 }
 GroomAssetTerminal {
  type=FGroomAssetTerminalDataflowNode_v2
  guid=#40
  inputPins [
   GuidesCollection {type=FManagedArrayCollection}
   StrandsCollection {type=FManagedArrayCollection}
   Spline Parameter {type=FCollectionAttributeKey}
   Spline Bones {type=FCollectionAttributeKey}
  ]
  outputPins []
  properties {AttributeNames=("Spline Parameter","Spline Bones")}
 }
 SplineToLinearSkinWeights {
  type=FSplineToLinearSkinWeightsDataflowNode
  guid=#41
  inputPins [
   Collection {type=FManagedArrayCollection}
   VertexSelection {type=FDataflowVertexSelection}
   SplineParamKey {type=FCollectionAttributeKey}
   SplineBonesKey {type=FCollectionAttributeKey}
  ]
  outputPins [
   Collection {type=FManagedArrayCollection}
   BoneIndicesKey {type=FCollectionAttributeKey}
   BoneWeightsKey {type=FCollectionAttributeKey}
  ]
 }
 EditSkinWeights {
  type=FDataflowCollectionEditSkinWeightsNode
  guid=#42
  inputPins [
   Collection {type=FManagedArrayCollection}
   BoneIndicesKey {type=FCollectionAttributeKey}
   BoneWeightsKey {type=FCollectionAttributeKey}
   SkeletalMesh {type="TObjectPtr<USkeletalMesh>"}
  ]
  outputPins [
   Collection {type=FManagedArrayCollection}
   BoneIndicesKey {type=FCollectionAttributeKey}
   BoneWeightsKey {type=FCollectionAttributeKey}
  ]
  properties {BoneIndicesName=BoneInd, BoneWeightsName=BoneW, SkeletalMesh=None, VertexGroup=(Name="Vertices"), bCompressSkinWeights=False}
 }
 LinearToSplineSkinWeights {
  type=FLinearToSplineSkinWeightsDataflowNode
  guid=#43
  inputPins [
   Collection {type=FManagedArrayCollection}
   VertexSelection {type=FDataflowVertexSelection}
   BoneIndicesKey {type=FCollectionAttributeKey}
   BoneWeightsKey {type=FCollectionAttributeKey}
  ]
  outputPins [
   Collection {type=FManagedArrayCollection}
   SplineParamKey {type=FCollectionAttributeKey}
   SplineBonesKey {type=FCollectionAttributeKey}
  ]
 }
 Skel_Mesh {
  type=FGetDataflowVariableNode
  guid=#44
  outputPins [Value {type="TObjectPtr<USkeletalMesh>"}]
  properties {
   Category=Skeletal Mesh
   Description=SkeletalMesh used to compute spline skinning weights. Will be stored onto the groom asset
   VariableName=Skel_Mesh
   VariablePropertyBag=(Value=/Engine/Transient.PropertyBag_2b7dfd4f8180b10(Skel_Mesh=None))
  }
 }
 GroomAssetToCollection_0 {
  type=FGroomAssetToCollectionDataflowNode
  guid=#45
  inputPins [
   GroomAsset {type="TObjectPtr<UGroomAsset>"}
   CurvesThickness {type=float}
   CurvesType {type=EGroomCollectionType}
  ]
  outputPins [Collection {type=FManagedArrayCollection}]
  properties {CurvesThickness=0.5, CurvesType=Guides, GroomAsset=None}
 }
 ReRouteNode_v1 {
  type=FDataflowReRouteNode
  guid=#46
  inputPins [Value {type="TObjectPtr<USkeletalMesh>"}]
  outputPins [Value {type="TObjectPtr<USkeletalMesh>"}]
 }
]
