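# Groom Dataflow template that builds hair joints from groom guides, producing a skeletal mesh and skeleton.
# Guides become a collection, pass through a curve selection tool and are resampled to few points; GuidesToJoints turns the selection into "Hair" bones parented to "Head".
# An input skeletal mesh supplies base geometry, and appended placeholder strings form both output asset paths.
# A path-rename branch derived from that mesh's path feeds nothing downstream.
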Dataflow /HairStrands/Dataflow/Templates/DF_Hair_JointCreation_Template
name=DF_Hair_JointCreation_Template
connections [
 {
  fromNode=GetGroomAsset
  fromPin=GroomAsset
  fromPinGuid=#1
  toNode=GroomAssetToCollection
  toPin=GroomAsset
  toPinGuid=#2
 }
 {
  fromNode=GroomAssetToCollection
  fromPin=Collection
  fromPinGuid=#3
  toNode=DataflowSelectionToolNode_v1
  toPin=Collection
  toPinGuid=#4
 }
 {
  fromNode=ReRouteNode_v1_0
  fromPin=Value
  fromPinGuid=#5
  toNode=ReRouteNode_v1
  toPin=Value
  toPinGuid=#6
 }
 {
  fromNode=DataflowSelectionToolNode_v1
  fromPin=Collection
  fromPinGuid=#7
  toNode=ResampleCurvePoints
  toPin=Collection
  toPinGuid=#8
 }
 {
  fromNode=DataflowSelectionToolNode_v1
  fromPin=Selection
  fromPinGuid=#9
  toNode=GuidesToJoints
  toPin=CurveSelection
  toPinGuid=#10
 }
 {
  fromNode=SkeletalMesh
  fromPin=SkeletalMesh
  fromPinGuid=#11
  toNode=ReRouteNode_v1_0
  toPin=Value
  toPinGuid=#12
 }
 {
  fromNode=SkeletalMesh
  fromPin=SkeletalMesh
  fromPinGuid=#11
  toNode=GetAssetPath
  toPin=Asset
  toPinGuid=#13
 }
 {
  fromNode=GetAssetPath
  fromPin=Path
  fromPinGuid=#14
  toNode=SkelMesh_Path
  toPin=Path
  toPinGuid=#15
 }
 {
  fromNode=GetAssetPath
  fromPin=Path
  fromPinGuid=#14
  toNode=Skeleton_Path
  toPin=Path
  toPinGuid=#16
 }
 {
  fromNode=ResampleCurvePoints
  fromPin=Collection
  fromPinGuid=#17
  toNode=GuidesToJoints
  toPin=Collection
  toPinGuid=#18
 }
 {
  fromNode=Skeletal Asset Name
  fromPin=String
  fromPinGuid=#19
  toNode=StringAppend
  toPin="Inputs[1]"
  toPinGuid=#20
 }
 {
  fromNode=StringAppend
  fromPin=String
  fromPinGuid=#21
  toNode=GuidesToJoints
  toPin=SkeletalMeshAssetPath
  toPinGuid=#22
 }
 {
  fromNode=ReRouteNode_v1
  fromPin=Value
  fromPinGuid=#23
  toNode=GuidesToJoints
  toPin=BaseSkeletalMesh
  toPinGuid=#24
 }
 {
  fromNode=Base Path
  fromPin=String
  fromPinGuid=#25
  toNode=StringAppend
  toPin="Inputs[0]"
  toPinGuid=#26
 }
 {
  fromNode=Skeleton Name
  fromPin=String
  fromPinGuid=#27
  toNode=StringAppend_0
  toPin="Inputs[1]"
  toPinGuid=#28
 }
 {
  fromNode=Base Path
  fromPin=String
  fromPinGuid=#25
  toNode=StringAppend_0
  toPin="Inputs[0]"
  toPinGuid=#29
 }
 {
  fromNode=StringAppend_0
  fromPin=String
  fromPinGuid=#30
  toNode=GuidesToJoints
  toPin=SkeletonAssetPath
  toPinGuid=#31
 }
]
nodes [
 GetGroomAsset {
  type=FGetGroomAssetDataflowNode_v2
  guid=#32
  inputPins []
  outputPins [GroomAsset {type="TObjectPtr<UGroomAsset>"}]
  properties {GroomAsset=None}
 }
 GroomAssetToCollection {
  type=FGroomAssetToCollectionDataflowNode
  guid=#33
  inputPins [
   GroomAsset {type="TObjectPtr<UGroomAsset>"}
   CurvesThickness {type=float}
   CurvesType {type=EGroomCollectionType}
  ]
  outputPins [Collection {type=FManagedArrayCollection}]
  properties {CurvesThickness=0.5, CurvesType=Guides, GroomAsset=None}
 }
 GuidesToJoints {
  type=FDataflowGuidesToJointsNode
  guid=#34
  inputPins [
   Collection {type=FManagedArrayCollection}
   CurveSelection {type=FDataflowCurveSelection}
   ParentBone {type=FDataflowStringTypes}
   BoneBaseName {type=FDataflowStringTypes}
   SkeletalMeshAssetPath {type=FString}
   SkeletonAssetPath {type=FString}
   BaseSkeletalMesh {type="TObjectPtr<USkeletalMesh>"}
  ]
  outputPins [
   SkeletalMeshAsset {type="TObjectPtr<USkeletalMesh>"}
   SkeletonAsset {type="TObjectPtr<USkeleton>"}
  ]
  properties {
   BaseSkeletalMesh=None
   BoneBaseName=(Value="Hair")
   ParentBone=(Value="Head")
   SkeletalMeshAssetPath=""
   SkeletonAssetPath=""
   bCopyGeometryFromBaseSkeletalMesh=True
  }
 }
 ResampleCurvePoints {
  type=FResampleCurvePointsDataflowNode
  guid=#35
  inputPins [
   Collection {type=FManagedArrayCollection}
   CurveSelection {type=FDataflowCurveSelection}
   NumPoints {type=int32}
  ]
  outputPins [Collection {type=FManagedArrayCollection}]
  properties {NumPoints=5, PointsCount=Size8, bResampleForSimulation=False}
 }
 GetAssetPath {
  type=FDataflowGetAssetPathNode
  guid=#36
  inputPins [Asset {type="TObjectPtr<USkeletalMesh>"}]
  outputPins [Path {type=FString}]
  properties {Asset=()}
 }
 SkelMesh_Path {
  type=FDataflowPathReplaceNode
  guid=#37
  inputPins [Path {type=FString}]
  outputPins [Path {type=FString}]
  properties {NameRules=((From="Helmet",To="NewJnts")), PathRules=""}
 }
 Skeleton_Path {
  type=FDataflowPathReplaceNode
  guid=#38
  inputPins [Path {type=FString}]
  outputPins [Path {type=FString}]
  properties {NameRules=((From="Helmet",To="NewJnts_Skel")), PathRules=""}
 }
 SkeletalMesh {
  type=FGetSkeletalMeshDataflowNode
  guid=#39
  inputPins []
  outputPins [SkeletalMesh {type="TObjectPtr<USkeletalMesh>"}]
  properties {PropertyName=SkeletalMesh, SkeletalMesh=None}
 }
 ReRouteNode_v1 {
  type=FDataflowReRouteNode
  guid=#40
  inputPins [Value {type="TObjectPtr<USkeletalMesh>"}]
  outputPins [Value {type="TObjectPtr<USkeletalMesh>"}]
 }
 ReRouteNode_v1_0 {
  type=FDataflowReRouteNode
  guid=#41
  inputPins [Value {type="TObjectPtr<USkeletalMesh>"}]
  outputPins [Value {type="TObjectPtr<USkeletalMesh>"}]
 }
 DataflowSelectionToolNode_v1 {
  type=FDataflowSelectionToolNode
  guid=#42
  inputPins [Collection {type=FManagedArrayCollection}]
  outputPins [
   Collection {type=FManagedArrayCollection}
   Selection {type=FDataflowCurveSelection}
  ]
 }
 Base Path {
  type=FMakeLiteralStringDataflowNode_v2
  guid=#43
  inputPins []
  outputPins [String {type=FString}]
  properties {String=/Game/}
 }
 StringAppend {
  type=FStringAppendDataflowNode_v2
  guid=#44
  inputPins [
   "Inputs[0]" {type=FString}
   "Inputs[1]" {type=FString}
  ]
  outputPins [String {type=FString}]
 }
 Skeletal Asset Name {
  type=FMakeLiteralStringDataflowNode_v2
  guid=#45
  inputPins []
  outputPins [String {type=FString}]
  properties {String=AssetName}
 }
 Skeleton Name {
  type=FMakeLiteralStringDataflowNode_v2
  guid=#46
  inputPins []
  outputPins [String {type=FString}]
  properties {String=AssetSkeletonName}
 }
 StringAppend_0 {
  type=FStringAppendDataflowNode_v2
  guid=#47
  inputPins [
   "Inputs[0]" {type=FString}
   "Inputs[1]" {type=FString}
  ]
  outputPins [String {type=FString}]
 }
]
